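# Physics asset for the male medium NinjaConvert player body, covering its back flap and hanging belt pieces.
# Capsules cover pelvis, lower spine, thighs, hands and the flap and belt bases.
# Box chains form left and right flaps (four links) and belt ends (three links), with angular limits tightest at the flap roots.
# Angularly free constraints join both bases to spine_01 and cross-link the two flap chains at their second and fourth links.
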PhysicsAsset /BRCosmetics/Characters/Player/Male/Medium/Bodies/M_MED_NinjaConvert/Meshes/M_Med_NinjaConvert_Physics
bodies {
 backFlap_base [{shapeType=Capsule, length=7.8696, radius=3.7446, center=<-0.8371 0.8414 0.0609>}]
 backFlap_l_01 [{shapeType=Box, extentX=10, extentY=2.6214, extentZ=10, center=<3.0965 0 0>}]
 backFlap_l_02 [{shapeType=Box, extentX=10, extentY=2, extentZ=10, center=<5 0 0>}]
 backFlap_l_03 [{shapeType=Box, extentX=10, extentY=2, extentZ=10}]
 backFlap_l_04 [{shapeType=Box, extentX=10, extentY=2, extentZ=10, center=<-0.016 0 0>}]
 backFlap_r_01 [{shapeType=Box, extentX=10, extentY=2.6214, extentZ=10, center=<3.1012 0 0>}]
 backFlap_r_02 [{shapeType=Box, extentX=10, extentY=2, extentZ=10, center=<5 0 0>}]
 backFlap_r_03 [{shapeType=Box, extentX=10, extentY=2, extentZ=10, center=<3.1374 0 0>}]
 backFlap_r_04 [{shapeType=Box, extentX=10, extentY=2, extentZ=10, center=<-0.0476 0 0>}]
 belt_base [{shapeType=Capsule, length=3.5644, radius=2.6573, center=<0.401 -0.2727 0.1223>}]
 dyn_belt_l_01 [{shapeType=Box, extentX=7, extentY=2, extentZ=5, center=<4.0928 0 0>}]
 dyn_belt_l_02 [{shapeType=Box, extentX=7, extentY=2, extentZ=5, center=<4.8178 0 0>}]
 dyn_belt_l_03 [{shapeType=Box, extentX=7, extentY=2, extentZ=5, center=<1.9002 0 0>}]
 dyn_belt_r_01 [{shapeType=Box, extentX=5, extentY=2, extentZ=5, center=<2.6549 0 0>}]
 dyn_belt_r_02 [{shapeType=Box, extentX=5, extentY=2, extentZ=5, center=<2.311 0 0>}]
 dyn_belt_r_03 [{shapeType=Box, extentX=5, extentY=3.3729, extentZ=5, center=<0.4336 0 0>}]
 hand_l [{shapeType=Capsule, length=8.7886, radius=7.1171, center=<7.7607 -2.7276 0.0682>, rotation=<rot 0 88.7498 -62.1737>}]
 hand_r [{shapeType=Capsule, length=8.8008, radius=7.0342, center=<-7.7633 2.7306 -0.0901>, rotation=<rot 0 88.6869 -63.1994>}]
 pelvis [{shapeType=Capsule, length=8.6007, radius=11.6248, center=<-0.0532 0.2405 0>, rotation=<rot 0 0 -0.1245>}]
 spine_01 [{shapeType=Capsule, length=10.6155, radius=9.3895, center=<5.7581 0.2884 0>, rotation=<rot 0 -90 -0.1245>}]
 thigh_l [{shapeType=Capsule, length=20.7179, radius=9.9201, center=<-20.3971 -0.0003 0>, rotation=<rot 0 90 -90>}]
 thigh_r [{shapeType=Capsule, length=22.4679, radius=8.1701, center=<19.8655 0.6463 0.5061>, rotation=<rot 0 94.1367 -89.873>}]
}
constraints [
 {bone1Name=hand_l, bone2Name=spine_01, swing1LimitDegrees=45, swing2LimitDegrees=45, twistLimitDegrees=45}
 {bone1Name=hand_r, bone2Name=spine_01, swing1LimitDegrees=45, swing2LimitDegrees=45, twistLimitDegrees=45}
 {bone1Name=thigh_l, bone2Name=pelvis, swing1LimitDegrees=45, swing2LimitDegrees=45, twistLimitDegrees=45}
 {bone1Name=thigh_r, bone2Name=pelvis, swing1LimitDegrees=45, swing2LimitDegrees=45, twistLimitDegrees=45}
 {
  bone1Name=spine_01
  bone2Name=pelvis
  swing1LimitDegrees=45
  swing1Motion=Free
  swing2LimitDegrees=45
  swing2Motion=Free
  twistLimitDegrees=45
  twistMotion=Free
 }
 {
  bone1Name=backFlap_base
  bone2Name=spine_01
  swing1LimitDegrees=45
  swing1Motion=Free
  swing2LimitDegrees=45
  swing2Motion=Free
  twistLimitDegrees=45
  twistMotion=Free
 }
 {bone1Name=backFlap_l_01, bone2Name=backFlap_base, swing1LimitDegrees=45, swing2LimitDegrees=5, twistLimitDegrees=3}
 {bone1Name=backFlap_l_02, bone2Name=backFlap_l_01, swing1LimitDegrees=45, swing2LimitDegrees=45, twistLimitDegrees=15}
 {bone1Name=backFlap_l_03, bone2Name=backFlap_l_02, swing1LimitDegrees=45, swing2LimitDegrees=45, twistLimitDegrees=15}
 {bone1Name=backFlap_l_04, bone2Name=backFlap_l_03, swing1LimitDegrees=45, swing2LimitDegrees=45, twistLimitDegrees=15}
 {bone1Name=backFlap_r_01, bone2Name=backFlap_base, swing1LimitDegrees=45, swing2LimitDegrees=5, twistLimitDegrees=3}
 {bone1Name=backFlap_r_02, bone2Name=backFlap_r_01, swing1LimitDegrees=45, swing2LimitDegrees=45, twistLimitDegrees=15}
 {bone1Name=backFlap_r_03, bone2Name=backFlap_r_02, swing1LimitDegrees=45, swing2LimitDegrees=45, twistLimitDegrees=15}
 {bone1Name=backFlap_r_04, bone2Name=backFlap_r_03, swing1LimitDegrees=45, swing2LimitDegrees=45, twistLimitDegrees=15}
 {
  bone1Name=belt_base
  bone2Name=spine_01
  swing1LimitDegrees=45
  swing1Motion=Free
  swing2LimitDegrees=45
  swing2Motion=Free
  twistLimitDegrees=45
  twistMotion=Free
 }
 {bone1Name=dyn_belt_l_01, bone2Name=belt_base, swing1LimitDegrees=45, swing2LimitDegrees=45, twistLimitDegrees=15}
 {bone1Name=dyn_belt_l_02, bone2Name=dyn_belt_l_01, swing1LimitDegrees=45, swing2LimitDegrees=45, twistLimitDegrees=15}
 {bone1Name=dyn_belt_l_03, bone2Name=dyn_belt_l_02, swing1LimitDegrees=45, swing2LimitDegrees=45, twistLimitDegrees=15}
 {bone1Name=dyn_belt_r_01, bone2Name=belt_base, swing1LimitDegrees=45, swing2LimitDegrees=45, twistLimitDegrees=15}
 {bone1Name=dyn_belt_r_02, bone2Name=dyn_belt_r_01, swing1LimitDegrees=45, swing2LimitDegrees=45, twistLimitDegrees=15}
 {bone1Name=dyn_belt_r_03, bone2Name=dyn_belt_r_02, swing1LimitDegrees=45, swing2LimitDegrees=45, twistLimitDegrees=15}
 {
  bone1Name=backFlap_l_02
  bone2Name=backFlap_r_02
  swing1LimitDegrees=45
  swing1Motion=Free
  swing2LimitDegrees=45
  swing2Motion=Free
  twistLimitDegrees=45
  twistMotion=Free
 }
 {
  bone1Name=backFlap_l_04
  bone2Name=backFlap_r_04
  swing1LimitDegrees=45
  swing1Motion=Free
  swing2LimitDegrees=45
  swing2Motion=Free
  twistLimitDegrees=45
  twistMotion=Free
 }
]
