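# Anim Blueprint for the NinjaConvert male body, adding secondary motion to its belt, scarf, shoulder pad and back flaps.
# It copies the attached parent's pose, then applies rotation constraints, AnimDynamics, a Look At and a final RigidBody pass with skydiving force.
# Additive bone offsets adjust the shoulder pad for skydiving and left-arm raise, which the EventGraph computes, and the flap base for crouching and skydiving.
# Crouching also enables a second, wider-limit set of flap simulations.
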AnimBlueprint /BRCosmetics/Characters/Player/Male/Medium/Bodies/M_MED_NinjaConvert/Meshes/M_Med_NinjaConvert_AnimBlueprint
target_skeleton=&/M_Med_NinjaConvert_Skeleton
graphs [
 AnimGraph {nodes [
   n1 = "Constraint - dyn_belt_l_01 (RR)"!(
    ComponentPose=Local To Component(LocalPose=Copy Pose From Mesh() {bCopyCurves=true, bUseAttachedParent=true})
    Alpha=1
    ConstraintWeights_0=0.6
    ConstraintWeights_1:float=0.4
   ) -> (Pose) {
    alphaBoolBlend {blendOption=Linear}
    boneToModify {boneName=dyn_belt_l_01}
    constraintSetup [{transformType=Rotation, targetBone {boneName=thigh_l}}, {transformType=Rotation, targetBone {boneName=thigh_r}}]
    constraintWeights [1,1]
   }
   n2 = "Constraint - dyn_belt_r_01 (RR)"!(ComponentPose=@n1, Alpha=1, ConstraintWeights_0=0.4, ConstraintWeights_1:float=0.6) -> (Pose) {
    alphaBoolBlend {blendOption=Linear}
    boneToModify {boneName=dyn_belt_r_01}
    constraintSetup [{transformType=Rotation, targetBone {boneName=thigh_l}}, {transformType=Rotation, targetBone {boneName=thigh_r}}]
    constraintWeights [1,1]
   }
   n3 = Anim Dynamics - Bone: dyn_scarf_front(ComponentPose=@n2, Alpha=1, ExternalForce="0,0,0") {
    bGravityOverrideInSimSpace=true
    bLinearSpring=true
    gravityScale=0.5
    linearSpringConstant=10
    alphaBoolBlend {blendOption=Linear}
    boundBone {boneName=dyn_scarf_front}
    physicsBodyDefinitions [{
      sphereCollisionRadius=0
      boundBone {boneName=dyn_scarf_front}
      boxExtents=<10 10 10>
      constraintSetup {angularLimitsMax=<0 0 0>, angularLimitsMin=<0 0 0>, angularTarget=<1 0 0>, linearAxesMax=<1 3 2>, linearAxesMin=<-1 0 -1>}
      localJointOffset=<0 0 0>
     }]
   }
   n4 = Anim Dynamics - Bone: dyn_scarf_back(ComponentPose=@n3, Alpha=1, ExternalForce="0,0,0") {
    bGravityOverrideInSimSpace=true
    bLinearSpring=true
    gravityScale=0.5
    linearSpringConstant=10
    alphaBoolBlend {blendOption=Linear}
    boundBone {boneName=dyn_scarf_back}
    physicsBodyDefinitions [{
      sphereCollisionRadius=0
      boundBone {boneName=dyn_scarf_back}
      boxExtents=<10 10 10>
      constraintSetup {angularLimitsMax=<0 0 0>, angularLimitsMin=<0 0 0>, angularTarget=<1 0 0>, linearAxesMax=<1 0 2>, linearAxesMin=<-1 -3 -1>}
      localJointOffset=<0 0 0>
     }]
   }
   n5 = Anim Dynamics - Bone: dyn_shldr_pad_01(ComponentPose=@n4, Alpha=0.25, ExternalForce="0,0,0") {
    alpha=0.9
    bGravityOverrideInSimSpace=true
    numSolverIterationsPostUpdate=4
    numSolverIterationsPreUpdate=16
    alphaBoolBlend {blendOption=Linear}
    boundBone {boneName=dyn_shldr_pad_01}
    physicsBodyDefinitions [{
      sphereCollisionRadius=0
      boundBone {boneName=dyn_shldr_pad_01}
      boxExtents=<10 10 10>
      constraintSetup {angularLimitsMax=<0 0 0>, angularLimitsMin=<0 0 0>, angularTarget=<0 0 0>, linearAxesMax=<1 1 1>, linearAxesMin=<-1 0 0>}
      localJointOffset=<0 0 -0.1>
     }]
   }
   n6 = Anim Dynamics - Bone: dyn_shldr_pad_02(ComponentPose=@n5, Alpha=0.25, ExternalForce="0,0,0") {
    alpha=0.9
    bGravityOverrideInSimSpace=true
    numSolverIterationsPostUpdate=4
    numSolverIterationsPreUpdate=16
    alphaBoolBlend {blendOption=Linear}
    boundBone {boneName=dyn_shldr_pad_02}
    physicsBodyDefinitions [{
      sphereCollisionRadius=0
      boundBone {boneName=dyn_shldr_pad_02}
      boxExtents=<10 10 10>
      constraintSetup {angularLimitsMax=<0 0 0>, angularLimitsMin=<0 0 0>, angularTarget=<0 0 0>, linearAxesMax=<1 1 1>, linearAxesMin=<-1 0 0>}
      localJointOffset=<0 0 -0.1>
     }]
   }
   n7 = Anim Dynamics - Bone: dyn_shldr_pad_03(ComponentPose=@n6, Alpha=0.25, ExternalForce="0,0,0") {
    alpha=0.9
    bGravityOverrideInSimSpace=true
    numSolverIterationsPostUpdate=4
    numSolverIterationsPreUpdate=16
    alphaBoolBlend {blendOption=Linear}
    boundBone {boneName=dyn_shldr_pad_03}
    physicsBodyDefinitions [{
      sphereCollisionRadius=0
      boundBone {boneName=dyn_shldr_pad_03}
      boxExtents=<10 10 10>
      constraintSetup {angularLimitsMax=<0 0 0>, angularLimitsMin=<0 0 0>, angularTarget=<0 0 0>, linearAxesMax=<1 1 1>, linearAxesMin=<-1 0 0>}
      localJointOffset=<0 0 -0.1>
     }]
   }
   n8 = Look At - Bone: dyn_shldr_pad_base(ComponentPose=@n7, Alpha=1) {
    alphaBoolBlend {blendOption=Linear}
    boneToModify {boneName=dyn_shldr_pad_base}
    lookAtTarget {boneReference {boneName=upperarm_twist_01_l}}
    lookAt_Axis {axis=<1 0>}
   }
   n9 = "Transform (Modify) Bone - Bone: dyn_shldr_pad_base"!(ComponentPose=@n8, bAlphaBoolEnabled:bool=Get skydiving() -> (skydiving:bool), Translation="5,0,3", Rotation="-15,0,0", Scale="1,1,1") -> (Pose) {
    alphaInputType=Bool
    rotationMode=BMM_Additive
    rotationSpace=BCS_BoneSpace
    translationMode=BMM_Additive
    translationSpace=BCS_BoneSpace
    alphaBoolBlend {blendInTime=0.5, blendOption=Linear, blendOutTime=0.54}
    boneToModify {boneName=dyn_shldr_pad_base}
    rotation {pitch=-15}
    translation {x=5, z=3}
   }
   n10 = "Transform (Modify) Bone - Bone: dyn_shldr_pad_base"!(
    ComponentPose=@n9
    Alpha:float=Get upperarm_l_raise_alpha() -> (upperarm_l_raise_alpha:double)
    Translation="0,0,0"
    Rotation="0,0,-30"
    Scale="1,1,1"
   ) -> (Pose) {
    rotationMode=BMM_Additive
    rotationSpace=BCS_BoneSpace
    translationMode=BMM_Additive
    translationSpace=BCS_BoneSpace
    alphaBoolBlend {blendInTime=0.5, blendOption=Linear, blendOutTime=0.54}
    alphaScaleBiasClamp {bInterpResult=true, interpSpeedDecreasing=15, interpSpeedIncreasing=15}
    boneToModify {boneName=dyn_shldr_pad_base}
    rotation {roll=-30}
   }
   n11 = Anim Dynamics - Bone: backFlap_l_01(ComponentPose=@n10, Alpha=1, ExternalForce="0,0,0") {
    angularDampingOverride=0.9999
    angularSpringConstant=100
    bGravityOverrideInSimSpace=true
    bOverrideAngularDamping=true
    bUseGravityOverride=true
    simulationSpace=BoneRelative
    alphaBoolBlend {blendOption=Linear}
    boundBone {boneName=backFlap_l_01}
    componentAppliedLinearAccClamp=<5000 5000 5000>
    componentLinearAccScale=<0.5 0.5 0.5>
    componentLinearVelScale=<0.5 0.5 0.5>
    gravityOverride {z=-980}
    physicsBodyDefinitions [{
      sphereCollisionRadius=0
      boundBone {boneName=backFlap_l_01}
      boxExtents=<10 15 15>
      constraintSetup {angularLimitsMax=<0 5 -5>, angularLimitsMin=<0 -5 -10>, angularTarget=<0 0 0>}
      localJointOffset=<6 0 0>
     }]
   }
   n12 = Anim Dynamics - Bone: backFlap_l_02(ComponentPose=@n11, Alpha=1, ExternalForce="0,0,0") {
    angularDampingOverride=0.9999
    angularSpringConstant=100
    bGravityOverrideInSimSpace=true
    bOverrideAngularDamping=true
    bUseGravityOverride=true
    simulationSpace=BoneRelative
    alphaBoolBlend {blendOption=Linear}
    boundBone {boneName=backFlap_l_02}
    componentAppliedLinearAccClamp=<5000 5000 5000>
    componentLinearAccScale=<0.5 0.5 0.5>
    componentLinearVelScale=<0.5 0.5 0.5>
    gravityOverride {z=-980}
    physicsBodyDefinitions [{
      sphereCollisionRadius=0
      boundBone {boneName=backFlap_l_02}
      boxExtents=<10 15 15>
      constraintSetup {angularLimitsMax=<0 5 5>, angularLimitsMin=<0 -5 -10>, angularTarget=<0 0 0>}
      localJointOffset=<6 0 0>
     }]
   }
   n13 = Anim Dynamics - Bone: backFlap_l_03(ComponentPose=@n12, Alpha=1, ExternalForce="0,0,0") {
    angularDampingOverride=0.9999
    angularSpringConstant=100
    bGravityOverrideInSimSpace=true
    bOverrideAngularDamping=true
    bUseGravityOverride=true
    simulationSpace=BoneRelative
    alphaBoolBlend {blendOption=Linear}
    boundBone {boneName=backFlap_l_03}
    componentAppliedLinearAccClamp=<5000 5000 5000>
    componentLinearAccScale=<0.5 0.5 0.5>
    componentLinearVelScale=<0.5 0.5 0.5>
    gravityOverride {z=-980}
    physicsBodyDefinitions [{
      sphereCollisionRadius=0
      boundBone {boneName=backFlap_l_03}
      boxExtents=<10 15 15>
      constraintSetup {angularLimitsMax=<0 5 5>, angularLimitsMin=<0 -5 -10>, angularTarget=<0 0 0>}
      localJointOffset=<6 0 0>
     }]
   }
   n14 = Anim Dynamics - Bone: backFlap_l_04(ComponentPose=@n13, Alpha=1, ExternalForce="0,0,0") {
    angularDampingOverride=0.9999
    angularSpringConstant=100
    bGravityOverrideInSimSpace=true
    bOverrideAngularDamping=true
    bUseGravityOverride=true
    simulationSpace=BoneRelative
    alphaBoolBlend {blendOption=Linear}
    boundBone {boneName=backFlap_l_04}
    componentAppliedLinearAccClamp=<5000 5000 5000>
    componentLinearAccScale=<0.5 0.5 0.5>
    componentLinearVelScale=<0.5 0.5 0.5>
    gravityOverride {z=-980}
    physicsBodyDefinitions [{
      sphereCollisionRadius=0
      boundBone {boneName=backFlap_l_04}
      boxExtents=<10 15 15>
      constraintSetup {angularLimitsMax=<0 5 10>, angularLimitsMin=<0 -5 -10>, angularTarget=<0 0 0>}
      localJointOffset=<6 0 0>
     }]
   }
   n15 = Anim Dynamics - Bone: backFlap_r_01(ComponentPose=@n14, Alpha=1, ExternalForce="0,0,0") {
    angularDampingOverride=0.9999
    angularSpringConstant=100
    bGravityOverrideInSimSpace=true
    bOverrideAngularDamping=true
    bUseGravityOverride=true
    simulationSpace=BoneRelative
    alphaBoolBlend {blendOption=Linear}
    boundBone {boneName=backFlap_r_01}
    componentAppliedLinearAccClamp=<5000 5000 5000>
    componentLinearAccScale=<0.5 0.5 0.5>
    componentLinearVelScale=<0.5 0.5 0.5>
    gravityOverride {z=-980}
    physicsBodyDefinitions [{
      sphereCollisionRadius=0
      boundBone {boneName=backFlap_r_01}
      boxExtents=<10 15 15>
      constraintSetup {angularLimitsMax=<0 5 -5>, angularLimitsMin=<0 -5 -10>, angularTarget=<0 0 0>}
      localJointOffset=<6 0 0>
     }]
   }
   n16 = Anim Dynamics - Bone: backFlap_r_02(ComponentPose=@n15, Alpha=1, ExternalForce="0,0,0") {
    angularDampingOverride=0.9999
    angularSpringConstant=100
    bGravityOverrideInSimSpace=true
    bOverrideAngularDamping=true
    bUseGravityOverride=true
    simulationSpace=BoneRelative
    alphaBoolBlend {blendOption=Linear}
    boundBone {boneName=backFlap_r_02}
    componentAppliedLinearAccClamp=<5000 5000 5000>
    componentLinearAccScale=<0.5 0.5 0.5>
    componentLinearVelScale=<0.5 0.5 0.5>
    gravityOverride {z=-980}
    physicsBodyDefinitions [{
      sphereCollisionRadius=0
      boundBone {boneName=backFlap_r_02}
      boxExtents=<10 15 15>
      constraintSetup {angularLimitsMax=<0 5 5>, angularLimitsMin=<0 -5 -10>, angularTarget=<0 0 0>}
      localJointOffset=<6 0 0>
     }]
   }
   n17 = Anim Dynamics - Bone: backFlap_r_03(ComponentPose=@n16, Alpha=1, ExternalForce="0,0,0") {
    angularDampingOverride=0.9999
    angularSpringConstant=100
    bGravityOverrideInSimSpace=true
    bOverrideAngularDamping=true
    bUseGravityOverride=true
    simulationSpace=BoneRelative
    alphaBoolBlend {blendOption=Linear}
    boundBone {boneName=backFlap_r_03}
    componentAppliedLinearAccClamp=<5000 5000 5000>
    componentLinearAccScale=<0.5 0.5 0.5>
    componentLinearVelScale=<0.5 0.5 0.5>
    gravityOverride {z=-980}
    physicsBodyDefinitions [{
      sphereCollisionRadius=0
      boundBone {boneName=backFlap_r_03}
      boxExtents=<10 15 15>
      constraintSetup {angularLimitsMax=<0 5 5>, angularLimitsMin=<0 -5 -10>, angularTarget=<0 0 0>}
      localJointOffset=<6 0 0>
     }]
   }
   n18 = Anim Dynamics - Bone: backFlap_r_04(ComponentPose=@n17, Alpha=1, ExternalForce="0,0,0") {
    angularDampingOverride=0.9999
    angularSpringConstant=100
    bGravityOverrideInSimSpace=true
    bOverrideAngularDamping=true
    bUseGravityOverride=true
    simulationSpace=BoneRelative
    alphaBoolBlend {blendOption=Linear}
    boundBone {boneName=backFlap_r_04}
    componentAppliedLinearAccClamp=<5000 5000 5000>
    componentLinearAccScale=<0.5 0.5 0.5>
    componentLinearVelScale=<0.5 0.5 0.5>
    gravityOverride {z=-980}
    physicsBodyDefinitions [{
      sphereCollisionRadius=0
      boundBone {boneName=backFlap_r_04}
      boxExtents=<10 15 15>
      constraintSetup {angularLimitsMax=<0 5 10>, angularLimitsMin=<0 -5 -10>, angularTarget=<0 0 0>}
      localJointOffset=<6 0 0>
     }]
   }
   n19 = Get bIsCrouching()
   n20 = Anim Dynamics - Bone: backFlap_l_01!(ComponentPose=@n18, bAlphaBoolEnabled:bool=@n19, ExternalForce="0,0,0") -> (Pose) {
    alphaInputType=Bool
    angularDampingOverride=0.999
    angularSpringConstant=100
    bGravityOverrideInSimSpace=true
    bOverrideAngularDamping=true
    bUseGravityOverride=true
    simulationSpace=BoneRelative
    alphaBoolBlend {blendOption=Linear}
    boundBone {boneName=backFlap_l_01}
    componentAppliedLinearAccClamp=<5000 5000 5000>
    componentLinearAccScale=<0.5 0.5 0.5>
    componentLinearVelScale=<0.5 0.5 0.5>
    gravityOverride {z=-980}
    physicsBodyDefinitions [{
      sphereCollisionRadius=0
      boundBone {boneName=backFlap_l_01}
      boxExtents=<10 15 15>
      constraintSetup {angularLimitsMax=<0 5 10>, angularLimitsMin=<0 -5 -25>, angularTarget=<0 0 0>}
      localJointOffset=<6 0 0>
     }]
   }
   n21 = Anim Dynamics - Bone: backFlap_l_02!(ComponentPose=@n20, bAlphaBoolEnabled:bool=@n19, ExternalForce="0,0,0") -> (Pose) {
    alphaInputType=Bool
    angularDampingOverride=0.999
    angularSpringConstant=100
    bGravityOverrideInSimSpace=true
    bOverrideAngularDamping=true
    bUseGravityOverride=true
    simulationSpace=BoneRelative
    alphaBoolBlend {blendOption=Linear}
    boundBone {boneName=backFlap_l_02}
    componentAppliedLinearAccClamp=<5000 5000 5000>
    componentLinearAccScale=<0.5 0.5 0.5>
    componentLinearVelScale=<0.5 0.5 0.5>
    gravityOverride {z=-980}
    physicsBodyDefinitions [{
      sphereCollisionRadius=0
      boundBone {boneName=backFlap_l_02}
      boxExtents=<10 15 15>
      constraintSetup {angularLimitsMax=<0 5 5>, angularLimitsMin=<0 -5 -25>, angularTarget=<0 0 0>}
      localJointOffset=<6 0 0>
     }]
   }
   n22 = Anim Dynamics - Bone: backFlap_l_03!(ComponentPose=@n21, bAlphaBoolEnabled:bool=@n19, ExternalForce="0,0,0") -> (Pose) {
    alphaInputType=Bool
    angularDampingOverride=0.999
    angularSpringConstant=100
    bGravityOverrideInSimSpace=true
    bOverrideAngularDamping=true
    bUseGravityOverride=true
    simulationSpace=BoneRelative
    alphaBoolBlend {blendOption=Linear}
    boundBone {boneName=backFlap_l_03}
    componentAppliedLinearAccClamp=<5000 5000 5000>
    componentLinearAccScale=<0.5 0.5 0.5>
    componentLinearVelScale=<0.5 0.5 0.5>
    gravityOverride {z=-980}
    physicsBodyDefinitions [{
      sphereCollisionRadius=0
      boundBone {boneName=backFlap_l_03}
      boxExtents=<10 15 15>
      constraintSetup {angularLimitsMax=<0 5 5>, angularLimitsMin=<0 -5 -25>, angularTarget=<0 0 0>}
      localJointOffset=<6 0 0>
     }]
   }
   n23 = Anim Dynamics - Bone: backFlap_l_04!(ComponentPose=@n22, bAlphaBoolEnabled:bool=@n19, ExternalForce="0,0,0") -> (Pose) {
    alphaInputType=Bool
    angularDampingOverride=0.999
    angularSpringConstant=100
    bGravityOverrideInSimSpace=true
    bOverrideAngularDamping=true
    bUseGravityOverride=true
    simulationSpace=BoneRelative
    alphaBoolBlend {blendOption=Linear}
    boundBone {boneName=backFlap_l_04}
    componentAppliedLinearAccClamp=<5000 5000 5000>
    componentLinearAccScale=<0.5 0.5 0.5>
    componentLinearVelScale=<0.5 0.5 0.5>
    gravityOverride {z=-980}
    physicsBodyDefinitions [{
      sphereCollisionRadius=0
      boundBone {boneName=backFlap_l_04}
      boxExtents=<10 15 15>
      constraintSetup {angularLimitsMax=<0 5 10>, angularLimitsMin=<0 -5 -25>, angularTarget=<0 0 0>}
      localJointOffset=<6 0 0>
     }]
   }
   n24 = Anim Dynamics - Bone: backFlap_r_01!(ComponentPose=@n23, bAlphaBoolEnabled:bool=@n19, ExternalForce="0,0,0") -> (Pose) {
    alphaInputType=Bool
    angularDampingOverride=0.999
    angularSpringConstant=100
    bGravityOverrideInSimSpace=true
    bOverrideAngularDamping=true
    bUseGravityOverride=true
    simulationSpace=BoneRelative
    alphaBoolBlend {blendOption=Linear}
    boundBone {boneName=backFlap_r_01}
    componentAppliedLinearAccClamp=<5000 5000 5000>
    componentLinearAccScale=<0.5 0.5 0.5>
    componentLinearVelScale=<0.5 0.5 0.5>
    gravityOverride {z=-980}
    physicsBodyDefinitions [{
      sphereCollisionRadius=0
      boundBone {boneName=backFlap_r_01}
      boxExtents=<10 15 15>
      constraintSetup {angularLimitsMax=<0 5 10>, angularLimitsMin=<0 -5 -25>, angularTarget=<0 0 0>}
      localJointOffset=<6 0 0>
     }]
   }
   n25 = Anim Dynamics - Bone: backFlap_r_02!(ComponentPose=@n24, bAlphaBoolEnabled:bool=@n19, ExternalForce="0,0,0") -> (Pose) {
    alphaInputType=Bool
    angularDampingOverride=0.999
    angularSpringConstant=100
    bGravityOverrideInSimSpace=true
    bOverrideAngularDamping=true
    bUseGravityOverride=true
    simulationSpace=BoneRelative
    alphaBoolBlend {blendOption=Linear}
    boundBone {boneName=backFlap_r_02}
    componentAppliedLinearAccClamp=<5000 5000 5000>
    componentLinearAccScale=<0.5 0.5 0.5>
    componentLinearVelScale=<0.5 0.5 0.5>
    gravityOverride {z=-980}
    physicsBodyDefinitions [{
      sphereCollisionRadius=0
      boundBone {boneName=backFlap_r_02}
      boxExtents=<10 15 15>
      constraintSetup {angularLimitsMax=<0 5 5>, angularLimitsMin=<0 -5 -25>, angularTarget=<0 0 0>}
      localJointOffset=<6 0 0>
     }]
   }
   n26 = Anim Dynamics - Bone: backFlap_r_03!(ComponentPose=@n25, bAlphaBoolEnabled:bool=@n19, ExternalForce="0,0,0") -> (Pose) {
    alphaInputType=Bool
    angularDampingOverride=0.999
    angularSpringConstant=100
    bGravityOverrideInSimSpace=true
    bOverrideAngularDamping=true
    bUseGravityOverride=true
    simulationSpace=BoneRelative
    alphaBoolBlend {blendOption=Linear}
    boundBone {boneName=backFlap_r_03}
    componentAppliedLinearAccClamp=<5000 5000 5000>
    componentLinearAccScale=<0.5 0.5 0.5>
    componentLinearVelScale=<0.5 0.5 0.5>
    gravityOverride {z=-980}
    physicsBodyDefinitions [{
      sphereCollisionRadius=0
      boundBone {boneName=backFlap_r_03}
      boxExtents=<10 15 15>
      constraintSetup {angularLimitsMax=<0 5 5>, angularLimitsMin=<0 -5 -25>, angularTarget=<0 0 0>}
      localJointOffset=<6 0 0>
     }]
   }
   n27 = Anim Dynamics - Bone: backFlap_r_04!(ComponentPose=@n26, bAlphaBoolEnabled:bool=@n19, ExternalForce="0,0,0") -> (Pose) {
    alphaInputType=Bool
    angularDampingOverride=0.999
    angularSpringConstant=100
    bGravityOverrideInSimSpace=true
    bOverrideAngularDamping=true
    bUseGravityOverride=true
    simulationSpace=BoneRelative
    alphaBoolBlend {blendOption=Linear}
    boundBone {boneName=backFlap_r_04}
    componentAppliedLinearAccClamp=<5000 5000 5000>
    componentLinearAccScale=<0.5 0.5 0.5>
    componentLinearVelScale=<0.5 0.5 0.5>
    gravityOverride {z=-980}
    physicsBodyDefinitions [{
      sphereCollisionRadius=0
      boundBone {boneName=backFlap_r_04}
      boxExtents=<10 15 15>
      constraintSetup {angularLimitsMax=<0 5 10>, angularLimitsMin=<0 -5 -25>, angularTarget=<0 0 0>}
      localJointOffset=<6 0 0>
     }]
   }
   n28 = "Transform (Modify) Bone - Bone: backFlap_base"!(ComponentPose=@n27, bAlphaBoolEnabled:bool=Get bIsCrouching(), Translation="0,-2,0", Rotation="0,0,0", Scale="1,1,1") -> (Pose) {
    alphaInputType=Bool
    rotationMode=BMM_Additive
    translationMode=BMM_Additive
    alphaBoolBlend {blendOption=Linear}
    boneToModify {boneName=backFlap_base}
    translation {y=-2}
   }
   n29 = "Transform (Modify) Bone - Bone: backFlap_base"(ComponentPose=@n28, bAlphaBoolEnabled=Get bIsSkyDiving(), Translation="0,-2,0", Rotation="0,0,30", Scale="1,1,1") {
    alphaInputType=Bool
    rotationMode=BMM_Additive
    translationMode=BMM_Additive
    alphaBoolBlend {blendOption=Linear}
    boneToModify {boneName=backFlap_base}
    rotation {roll=30}
    translation {y=-2}
   }
   n30 = RigidBody(
    ComponentPose=@n29
    ExternalForce=Break Fort Anim Input Skydiving External Force(FortAnimInput_SkydivingExternalForce=Get SkydivingExternalForceInput()) -> (
     bUseSkydivingVectorForce=~
     SkydivingVectorForce=~
     HeadToPelvisDirection=~
     FloatingMultiplier=~
     FloatingAdditive=~
     DivingMultiplier=~
     DivingAdditive=~
     ParachutingMultiplier=~
     ParachutingAdditive=~
    ).SkydivingVectorForce
   ) {
    bClampLinearTranslationLimitToRefPose=true
    bOverrideWorldGravity=true
    overridePhysicsAsset=&/M_Med_NinjaConvert_Physics
    simulationSpace=BaseBoneSpace
    alphaBoolBlend {blendOption=Linear}
    baseBoneRef {boneName=root}
    componentLinearAccScale=<0.2 0.2 0.2>
    overrideWorldGravity {z=-980}
   }
   Output Pose(Result=Component To Local(ComponentPose=@n30)) {layerGroup=None}
  ]}
 EventGraph {nodes [
   Event BlueprintUpdateAnimation()
   > Sequence()
   n1 = Get bIsSkydiveDiveMode()
   > n2 = Branch(Condition=OR Boolean!(A=Get bIsSkydiveFloating(), B=Get bIsParachuteOpened(), C:bool=@n1, D:bool=@n1) -> (ReturnValue))
   > Set skydiving(skydiving:bool=true) -> (Output_Get:bool)
   Set skydiving(execute=@n2.else, skydiving=false)
   n3 = GetOwnerSkeletalMeshComponent()
   n4 = BreakRotator(InRot=BreakTransform(InTransform=GetDeltaTransformFromRefPose(self=@n3, BoneName=upperarm_l, BaseName=clavicle_l)).Rotation)
   > Set upperarm_l_raise_alpha(upperarm_l_raise_alpha:double=MapRangeClamped(Value=@n4.Pitch, InRangeA=30, InRangeB=50)) -> (Output_Get:double)
   MapRangeClamped(Value=@n4, InRangeA=-50, InRangeB=0)
   MapRangeClamped(Value=@n4.Yaw, InRangeA=0, InRangeB=90)
  ]}
]
variables [
 skydiving {type=boolean}
 upperarm_l_raise_alpha {type=number}
]
